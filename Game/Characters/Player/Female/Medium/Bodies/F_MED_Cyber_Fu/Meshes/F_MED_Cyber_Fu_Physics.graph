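# Physics asset for the F_MED_Cyber_Fu female player body, covering the pelvis, spine, left forearm and two dangling-cable bones on that arm.
# Single capsules form the pelvis, two spine segments and forearm, while each cable bone uses three angled capsules to approximate a curve.
# Constraints chain pelvis through spine to a free-swinging, free-twisting forearm.
# Both cables attach to that forearm, with angular range on only one swing axis.
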PhysicsAsset /Game/Characters/Player/Female/Medium/Bodies/F_MED_Cyber_Fu/Meshes/F_MED_Cyber_Fu_Physics
bodies {
 dyn_cable_02_l [
  {shapeType=Capsule, length=4, radius=1.8, shapeName=dyn_cable_02_l_capsule, center=<3.7445 0 0.1072>, rotation=<rot 10 0.0002 0.0001>}
  {shapeType=Capsule, length=4.5, radius=1.8, shapeName=dyn_cable_02_l_capsule, center=<1.5587 0 -6.1422>, rotation=<rot -42.0287 0 0.0001>}
  {shapeType=Capsule, length=3, radius=1.8, shapeName=dyn_cable_02_l_capsule, center=<1.086 0 5.5212>, rotation=<rot 50 0.0007 0.0004>}
 ]
 dyn_cable_03_l [
  {shapeType=Capsule, length=4, radius=2, shapeName=dyn_cable_03_l_capsule, center=<-3.0807 0 6.0519>, rotation=<rot 60.1046 -3.4663 1.8451>}
  {shapeType=Capsule, length=5, radius=2, shapeName=dyn_cable_03_l_capsule, center=<3.0676 0 0.9185>, rotation=<rot 30 0 0>}
  {shapeType=Capsule, length=4, radius=2, shapeName=dyn_cable_03_l_capsule, center=<4.8343 0 -6.9037>, rotation=<rot -9.9997 0.0001 0.0002>}
 ]
 lowerarm_twist_02_l [{
   shapeType=Capsule
   length=20.9464
   radius=3.6707
   shapeName=lowerarm_twist_02_l_capsule
   center=<-0.0328 0 0.6671>
   rotation=<rot 87.1878 -179.9973 -179.9975>
  }]
 pelvis [{shapeType=Capsule, length=12.7402, radius=10, center=<0.6115 -1.8374 0.2204>, rotation=<rot 0 90 0>}]
 spine_01 [{shapeType=Capsule, length=9.8682, radius=12, center=<4.8416 1.6991 -0.1087>, rotation=<rot 90 0 0>}]
 spine_03 [{shapeType=Capsule, length=6.25, radius=10, center=<5.9298 4.7471 0.4704>, rotation=<rot 0 176.0619 -168.602>}]
}
constraints [
 {bone1Name=spine_01, bone2Name=pelvis, swing1LimitDegrees=45, swing2LimitDegrees=45, twistLimitDegrees=45}
 {bone1Name=spine_03, bone2Name=spine_01, swing1LimitDegrees=45, swing2LimitDegrees=45, twistLimitDegrees=45}
 {
  bone1Name=lowerarm_twist_02_l
  bone2Name=spine_03
  swing1LimitDegrees=45
  swing1Motion=Free
  swing2LimitDegrees=45
  swing2Motion=Free
  twistLimitDegrees=45
  twistMotion=Free
 }
 {bone1Name=dyn_cable_02_l, bone2Name=lowerarm_twist_02_l, swing1LimitDegrees=40, swing2LimitDegrees=0, twistLimitDegrees=0}
 {bone1Name=dyn_cable_03_l, bone2Name=lowerarm_twist_02_l, swing1LimitDegrees=40, swing2LimitDegrees=0, twistLimitDegrees=0}
]
