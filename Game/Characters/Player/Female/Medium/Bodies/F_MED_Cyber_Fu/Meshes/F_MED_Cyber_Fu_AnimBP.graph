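# Post-processing AnimBlueprint for the F_MED_Cyber_Fu female player body, adding deformation and secondary motion to the pose copied from the parent mesh.
# A Control Rig feeds a cached pose; left-forearm cable constraints apply unless skydiving or parachuting.
# Crouching offsets the left glute bone, then RigidBody simulation with overridden gravity and Anim Dynamics shoelace springs follow.
# The event graph updates air, crouch and gravity values, using a deprecated gravity function.
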AnimBlueprint /Game/Characters/Player/Female/Medium/Bodies/F_MED_Cyber_Fu/Meshes/F_MED_Cyber_Fu_AnimBP
target_skeleton=&/F_MED_Cyber_Fu_Skeleton
graphs [
 AnimGraph {nodes [
   n1 = "Constraint - dyn_cable_01_l (PPPP)"!(
    ComponentPose=Local To Component(LocalPose=Use cached pose 'Cache_StartPose'())
    Alpha=1
    ConstraintWeights_0=0.8305
    ConstraintWeights_1:float=0.0933
    ConstraintWeights_2:float=0.0592
    ConstraintWeights_3:float=1
   ) -> (Pose) {
    alphaBoolBlend {blendOption=Linear}
    boneToModify {boneName=dyn_cable_01_l}
    constraintSetup [
     {targetBone {boneName=lowerarm_twist_01_l}}
     {targetBone {boneName=lowerarm_l}}
     {targetBone {boneName=lowerarm_twist_02_l}}
     {targetBone {boneName=hand_l}}
    ]
    constraintWeights [1,1,1,1]
   }
   n2 = "Constraint - dyn_cable_01_l (R)"(ComponentPose=@n1, Alpha=0.5, ConstraintWeights_0=1) {
    alphaBoolBlend {blendOption=Linear}
    boneToModify {boneName=dyn_cable_01_l}
    constraintSetup [{transformType=Rotation, targetBone {boneName=dyn_cable_01_l_home}}]
    constraintWeights [1]
   }
   n3 = "Constraint - dyn_cable_04_l (PPPP)"!(
    ComponentPose=@n2
    Alpha=1
    ConstraintWeights_0=0.7392
    ConstraintWeights_1:float=0.2514
    ConstraintWeights_2:float=0.0083
    ConstraintWeights_3:float=0.0011
   ) -> (Pose) {
    alphaBoolBlend {blendOption=Linear}
    boneToModify {boneName=dyn_cable_04_l}
    constraintSetup [
     {targetBone {boneName=lowerarm_twist_01_l}}
     {targetBone {boneName=hand_l}}
     {targetBone {boneName=lowerarm_l}}
     {targetBone {boneName=thumb_01_l}}
    ]
    constraintWeights [1,1,1,1]
   }
   n4 = "Constraint - dyn_cable_04_l (R)"(ComponentPose=@n3, Alpha=0.5, ConstraintWeights_0=1) {
    alphaBoolBlend {blendOption=Linear}
    boneToModify {boneName=dyn_cable_04_l}
    constraintSetup [{transformType=Rotation, targetBone {boneName=dyn_cable_04_l_home}}]
    constraintWeights [1]
   }
   n5 = Component To Local(ComponentPose=@n4)
   n6 = Use cached pose 'Cache_StartPose'()
   n7 = Get IsInAir() -> (IsInAir:bool)
   n8 = "Transform (Modify) Bone - Bone: deform_glute_l"(
    ComponentPose=Local To Component(LocalPose=Blend Poses by bool(BlendPose_0=@n6, BlendPose_1=@n5, bActiveValue=@n7) {blendType=Linear})
    Alpha=Get IsCrouching() -> (IsCrouching:double)
    Translation="-10,0,20"
    Rotation="0,0,0"
    Scale="1,1,1"
   ) {
    translationMode=BMM_Additive
    translationSpace=BCS_BoneSpace
    alphaBoolBlend {blendOption=Linear}
    alphaScaleBiasClamp {bInterpResult=true}
    boneToModify {boneName=deform_glute_l}
    translation {x=-10, z=20}
   }
   n9 = RigidBody!(ComponentPose=@n8, Alpha, OverrideWorldGravity:Vector=Get GravityOverride(), ExternalForce) -> (Pose) {
    bOverrideWorldGravity=true
    overridePhysicsAsset=&/F_MED_Cyber_Fu_Physics
    simulationSpace=BaseBoneSpace
    alphaBoolBlend {blendOption=Linear}
    baseBoneRef {boneName=spine_01}
    componentAppliedLinearAccClamp=<3000 3000 3000>
    componentLinearAccScale=<0.2 0.2 0.2>
    componentLinearVelScale=<0.2 0.2 0.2>
    overrideWorldGravity {z=-980}
   }
   n10 = Anim Dynamics - Bone: dyn_shoelace_02_l(ComponentPose=@n9, Alpha=1, ExternalForce="0,0,0") {
    angularSpringConstant=25
    bAngularSpring=true
    bGravityOverrideInSimSpace=true
    gravityScale=0.15
    alphaBoolBlend {blendOption=Linear}
    boundBone {boneName=dyn_shoelace_02_l}
    componentAppliedLinearAccClamp=<1000 1000 1000>
    componentLinearAccScale=<1 1 1>
    componentLinearVelScale=<1 1 1>
    physicsBodyDefinitions [{
      sphereCollisionRadius=0
      boundBone {boneName=dyn_shoelace_02_l}
      boxExtents=<6 6 6>
      constraintSetup {angularLimitsMax=<5 25 3>, angularLimitsMin=<-5 -10 -3>, angularTarget=<0 0 0>}
      localJointOffset=<2 0 0>
     }]
   }
   n11 = Anim Dynamics - Bone: dyn_shoelace_03_l(ComponentPose=@n10, Alpha=1, ExternalForce="0,0,0") {
    angularSpringConstant=25
    bAngularSpring=true
    bGravityOverrideInSimSpace=true
    gravityScale=0.15
    alphaBoolBlend {blendOption=Linear}
    boundBone {boneName=dyn_shoelace_03_l}
    componentAppliedLinearAccClamp=<1000 1000 1000>
    componentLinearAccScale=<1 1 1>
    componentLinearVelScale=<1 1 1>
    physicsBodyDefinitions [{
      sphereCollisionRadius=0
      boundBone {boneName=dyn_shoelace_03_l}
      boxExtents=<6 6 6>
      constraintSetup {angularLimitsMax=<5 25 3>, angularLimitsMin=<-5 -10 -3>, angularTarget=<0 0 0>}
      localJointOffset=<2 0 0>
     }]
   }
   n12 = Anim Dynamics - Bone: dyn_shoelace_04_l(ComponentPose=@n11, Alpha=1, ExternalForce="0,0,0") {
    angularSpringConstant=25
    bAngularSpring=true
    bGravityOverrideInSimSpace=true
    gravityScale=0.15
    alphaBoolBlend {blendOption=Linear}
    boundBone {boneName=dyn_shoelace_04_l}
    componentAppliedLinearAccClamp=<1000 1000 1000>
    componentLinearAccScale=<1 1 1>
    componentLinearVelScale=<1 1 1>
    physicsBodyDefinitions [{
      sphereCollisionRadius=0
      boundBone {boneName=dyn_shoelace_04_l}
      boxExtents=<6 6 6>
      constraintSetup {angularLimitsMax=<5 25 3>, angularLimitsMin=<-5 -10 -3>, angularTarget=<0 0 0>}
      localJointOffset=<2 0 0>
     }]
   }
   n13 = Anim Dynamics - Bone: dyn_shoelace_05_l(ComponentPose=@n12, Alpha=1, ExternalForce="0,0,0") {
    angularSpringConstant=25
    bAngularSpring=true
    bGravityOverrideInSimSpace=true
    gravityScale=0.15
    alphaBoolBlend {blendOption=Linear}
    boundBone {boneName=dyn_shoelace_05_l}
    componentAppliedLinearAccClamp=<1000 1000 1000>
    componentLinearAccScale=<1 1 1>
    componentLinearVelScale=<1 1 1>
    physicsBodyDefinitions [{
      sphereCollisionRadius=0
      boundBone {boneName=dyn_shoelace_05_l}
      boxExtents=<6 6 6>
      constraintSetup {angularLimitsMax=<5 25 3>, angularLimitsMin=<-5 -10 -3>, angularTarget=<0 0 0>}
      localJointOffset=<2 0 0>
     }]
   }
   n14 = Anim Dynamics - Bone: dyn_shoelace_02_r(ComponentPose=@n13, Alpha=1, ExternalForce="0,0,0") {
    angularSpringConstant=25
    bAngularSpring=true
    bGravityOverrideInSimSpace=true
    gravityScale=0.15
    alphaBoolBlend {blendOption=Linear}
    boundBone {boneName=dyn_shoelace_02_r}
    componentAppliedLinearAccClamp=<1000 1000 1000>
    componentLinearAccScale=<1 1 1>
    componentLinearVelScale=<1 1 1>
    physicsBodyDefinitions [{
      sphereCollisionRadius=0
      boundBone {boneName=dyn_shoelace_02_r}
      boxExtents=<6 6 6>
      constraintSetup {angularLimitsMax=<5 25 3>, angularLimitsMin=<-5 -10 -3>, angularTarget=<0 0 0>}
      localJointOffset=<2 0 0>
     }]
   }
   n15 = Anim Dynamics - Bone: dyn_shoelace_03_r(ComponentPose=@n14, Alpha=1, ExternalForce="0,0,0") {
    angularSpringConstant=25
    bAngularSpring=true
    bGravityOverrideInSimSpace=true
    gravityScale=0.15
    alphaBoolBlend {blendOption=Linear}
    boundBone {boneName=dyn_shoelace_03_r}
    componentAppliedLinearAccClamp=<1000 1000 1000>
    componentLinearAccScale=<1 1 1>
    componentLinearVelScale=<1 1 1>
    physicsBodyDefinitions [{
      sphereCollisionRadius=0
      boundBone {boneName=dyn_shoelace_03_r}
      boxExtents=<6 6 6>
      constraintSetup {angularLimitsMax=<5 25 3>, angularLimitsMin=<-5 -10 -3>, angularTarget=<0 0 0>}
      localJointOffset=<2 0 0>
     }]
   }
   n16 = Anim Dynamics - Bone: dyn_shoelace_04_r(ComponentPose=@n15, Alpha=1, ExternalForce="0,0,0") {
    angularSpringConstant=25
    bAngularSpring=true
    bGravityOverrideInSimSpace=true
    gravityScale=0.15
    alphaBoolBlend {blendOption=Linear}
    boundBone {boneName=dyn_shoelace_04_r}
    componentAppliedLinearAccClamp=<1000 1000 1000>
    componentLinearAccScale=<1 1 1>
    componentLinearVelScale=<1 1 1>
    physicsBodyDefinitions [{
      sphereCollisionRadius=0
      boundBone {boneName=dyn_shoelace_04_r}
      boxExtents=<6 6 6>
      constraintSetup {angularLimitsMax=<5 25 3>, angularLimitsMin=<-5 -10 -3>, angularTarget=<0 0 0>}
      localJointOffset=<2 0 0>
     }]
   }
   n17 = Anim Dynamics - Bone: dyn_shoelace_05_r(ComponentPose=@n16, Alpha=1, ExternalForce="0,0,0") {
    angularSpringConstant=25
    bAngularSpring=true
    bGravityOverrideInSimSpace=true
    gravityScale=0.15
    alphaBoolBlend {blendOption=Linear}
    boundBone {boneName=dyn_shoelace_05_r}
    componentAppliedLinearAccClamp=<1000 1000 1000>
    componentLinearAccScale=<1 1 1>
    componentLinearVelScale=<1 1 1>
    physicsBodyDefinitions [{
      sphereCollisionRadius=0
      boundBone {boneName=dyn_shoelace_05_r}
      boxExtents=<6 6 6>
      constraintSetup {angularLimitsMax=<5 25 3>, angularLimitsMin=<-5 -10 -3>, angularTarget=<0 0 0>}
      localJointOffset=<2 0 0>
     }]
   }
   Output Pose(Result=Component To Local(ComponentPose=@n17)) {layerGroup=None}
   n18 = Control Rig(Source=Copy Pose From Mesh() {bUseAttachedParent=true}) {
    alphaBoolBlend {blendOption=Linear}
    controlRigAssetReference {blueprintRigClass=&/F_MED_DeformRig_Master_ControlRig_CyberFu.F_MED_DeformRig_Master_ControlRig_CyberFu_C}
   }
   Save cached pose 'Cache_StartPose'(Pose=@n18) {cacheName=Cache_StartPose}
  ]}
 EventGraph {nodes [
   Event BlueprintUpdateAnimation()
   > n1 = Sequence!(execute) -> (then_0, then_1, then_2:Exec)
   Set IsCrouching(execute=@n1.then_2, IsCrouching:double="To Float (Boolean)"(InBool=Get bIsCrouching())) -> (Output_Get:double)
   n2 = Gravity Override State Function DEPRECATED(
    execute=@n1.then_0
    CharacterPartAnimInstance=Self-Reference()
    GameTimeInSeconds=GetGameTimeInSeconds()
    GravityOverrideParams=Get Gravity Override Params()
   )
   > Set GravityOverride(GravityOverride=@n2)
   Set IsInAir(
    execute=@n1.then_1
    IsInAir:bool=OR Boolean!(A=Get bIsSkyDiving(), B=Get bIsSkydiveDiveMode(), C:bool=Get bIsSkydiveFloating(), D:bool=Get bIsParachuteOpened()) -> (ReturnValue)
   ) -> (Output_Get:bool)
   TryGetPawnOwner()
  ]}
]
variables [
 Gravity Override Params {type=GravityOverrideParamsStruct}
 GravityOverride {type=Vector}
 IsCrouching {type=number}
 IsInAir {type=boolean}
]
